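; Cosmetic variant script that picks the Doom cape back bling's style from the player's loadout.
; With the AnglePatch_Lodge outfit equipped, it compares both items' Progressive channels: matching Stage1 (base) returns Emissive1, matching Stage2 (Royal) returns Emissive2.
; Without that outfit, it returns Emissive3 only when the cape's Progressive variant is already Emissive3.
; Other combinations have no explicit return.
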(fn DetermineLoadoutVariantSelection
  (params (Loadout :type Cosmetic%20Loadout%20Structure%20%28by%20ref%29))
  (if (Class|FortCosmeticFunctionLibrary|IsItemInLoadout Loadout "/BRCosmetics/Athena/Items/Cosmetics/Characters/Character_AnglePatch_Lodge.Character_AnglePatch_Lodge")
    ; Base Doom outfit + base Doom backbling
    (if (== (Class|FortCosmeticFunctionLibrary|GetActiveVariantForSoftItemAndChannel Loadout "/BRCosmetics/Athena/Items/Cosmetics/Backpacks/Backpack_AnglePatch_Lodge_Cape.Backpack_AnglePatch_Lodge_Cape" "(TagName=\"Cosmetics.Variant.Channel.Progressive\")") "(TagName=\"Cosmetics.Variant.Property.Stage1\")")
      (if (== (Class|FortCosmeticFunctionLibrary|GetActiveVariantForSoftItemAndChannel Loadout "/BRCosmetics/Athena/Items/Cosmetics/Characters/Character_AnglePatch_Lodge.Character_AnglePatch_Lodge" "(TagName=\"Cosmetics.Variant.Channel.Progressive\")") "(TagName=\"Cosmetics.Variant.Property.Stage1\")")
        (switch Utilities|FlowControl|Switch|SwitchonGameplayTag 0
          (:Default
            (return "(TagName=\"Cosmetics.Variant.Property.Emissive1\")"))))
      (elif (== (Class|FortCosmeticFunctionLibrary|GetActiveVariantForSoftItemAndChannel Loadout "/BRCosmetics/Athena/Items/Cosmetics/Backpacks/Backpack_AnglePatch_Lodge_Cape.Backpack_AnglePatch_Lodge_Cape" "(TagName=\"Cosmetics.Variant.Channel.Progressive\")") "(TagName=\"Cosmetics.Variant.Property.Stage2\")")
        ; Royal Doom outfit + Royal Backbling
        (if (== (Class|FortCosmeticFunctionLibrary|GetActiveVariantForSoftItemAndChannel Loadout "/BRCosmetics/Athena/Items/Cosmetics/Characters/Character_AnglePatch_Lodge.Character_AnglePatch_Lodge" "(TagName=\"Cosmetics.Variant.Channel.Progressive\")") "(TagName=\"Cosmetics.Variant.Property.Stage2\")")
          (switch Utilities|FlowControl|Switch|SwitchonGameplayTag 0
            (:Default
              (return "(TagName=\"Cosmetics.Variant.Property.Emissive2\")"))))))
    (else
      (switch Utilities|FlowControl|Switch|SwitchonGameplayTag (Class|FortCosmeticFunctionLibrary|GetActiveVariantForSoftItemAndChannel Loadout "/BRCosmetics/Athena/Items/Cosmetics/Backpacks/Backpack_AnglePatch_Lodge_Cape.Backpack_AnglePatch_Lodge_Cape" "(TagName=\"Cosmetics.Variant.Channel.Progressive\")")
        (:Cosmetics.Variant.Property.Emissive3
          (return "(TagName=\"Cosmetics.Variant.Property.Emissive3\")"))))))
